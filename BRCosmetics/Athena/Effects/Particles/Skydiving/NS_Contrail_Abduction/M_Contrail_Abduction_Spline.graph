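# Unlit translucent, two-sided material for the spline meshes of the Abduction skydiving contrail, also usable on Niagara mesh particles.
# A tractor-beam ring curve from a curve atlas, read along UV X, is faded by view angle against the vertex tangent.
# That mask, tinted and scaled by particle color and alpha, shapes emissive and a translucency term.
# SplineThicken sets strand width from a SplineWidth parameter with slight per-particle random variation.
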Material /BRCosmetics/Athena/Effects/Particles/Skydiving/NS_Contrail_Abduction/M_Contrail_Abduction_Spline
graph {
 properties {
  bCastRayTracedShadows=false
  bUsedWithNiagaraMeshParticles=true
  bUsedWithSplineMeshes=true
  blendMode=BLEND_Translucent
  pixelDepthOffsetMode=PDOM_Legacy
  shadingModel=MSM_Unlit
  twoSided=true
  usedShadingModels=Unlit
 }
 nodes [
  n1 = CurveAtlasRowParameter(CurveTime=MaterialFunctionCall(Float2=TextureCoordinate()) {materialFunction=&BreakOutFloat2Components}.R) {atlas=/Game/Athena/LinearColorCurves/CA_Contrail_Abduction, curve=/Game/Athena/LinearColorCurves/Curve_Contrail_Abduction_TractorBeam_Ring}
  n2 = Abs(None=DotProduct(A=CameraVectorWS(), B=VertexTangentWS()))
  n3 = Saturate(None=Multiply(A=@n2) {constB=2})
  n4 = Multiply(A=Multiply(A=@n1.R, B=@n1.R), B=@n3)
  n5 = ParticleColor()
  MaterialFunctionCall(Emissive=Multiply(A=LinearInterpolate(Alpha=@n4) {constA=0.2}, B=@n5.RGB)) {materialFunction=&MF_FX_EmissiveAdjustments}
  n6 = Multiply(A=ScalarParameter() {defaultValue=10, parameterName=SplineWidth}, B=LinearInterpolate(Alpha=ParticleRandom()) {constA=0.8})
  MaterialFunctionCall(WidthBase=@n6, WidthTip=@n6) {materialFunction=&SplineThicken}
  Multiply(A=Multiply(A=@n1.R, B=@n3), B=@n5.A)
  Subtract(A=@n2) {constB=0.1}
 ]
}
