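# Niagara skydiving contrail cosmetic themed on alien abduction, with UFOs and tractor beams attached to the player's bones.
# UFO meshes wobble at a bone offset along the system's UfoVelocity and UFODistance, spawning smaller spiraling UFOs through a particle reader.
# Beam rings travel along that same offset, spawning faster and living shorter as Dive rises.
# A bone-flow aura, bone-attached flares, orbiting dive-scaled lights and surrounding streaks complete the effect.
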NiagaraSystem /BRCosmetics/Athena/Effects/Particles/Skydiving/NS_Contrail_Abduction/NS_Contrail_Abduction
dependencies {
 usedDataInterfaces [
  &NiagaraDataInterfaceCurve
  &NiagaraDataInterfaceArrayFloat
  &NiagaraDataInterfaceVectorCurve
  &NiagaraDataInterfaceParticleRead
  &NiagaraDataInterfaceColorCurve
 ]
 usedDynamicInputs [
  &MakeVector4_FromLinearColor
  &LerpLinearColors
  &FN_FrontEndFloatModifier
  &FN_FrontEndVector3Modifier
  &TransformVector
  &Lerp_Vector
  &SetBoolByFloatComparison
  &NormalizeVector
  &Multiply_Vector
 ]
 usedModules [
  &SystemState
  /Game/Effects/Niagara/Modules/Skydiving/FN_DivingAttributes/v5/FN_DivingAttributes
  &self:SystemSpawnScript.SystemScriptSource.SystemScriptGraph.NiagaraNodeAssignment_2.SetVariables_#1
  &self:SystemSpawnScript.SystemScriptSource.SystemScriptGraph.NiagaraNodeAssignment_1.SetVariables_#2
  &FN_Integrator
  &self:SystemSpawnScript.SystemScriptSource.SystemScriptGraph.NiagaraNodeAssignment_4.SetVariables_#3
  &self:SystemSpawnScript.SystemScriptSource.SystemScriptGraph.NiagaraNodeAssignment_7.SetVariables_#4
 ]
}
emitters {
 Aura {
  emitterUpdateScript [
   &EmitterState(Life Cycle Mode=::NewEnumerator1, Loop Duration=1, Recalculate Duration Each Loop=true, UseLoopDelay=false)
   &SpawnRate(SpawnRate=30, Use Spawn Probability=false)
  ]
  particleSpawnScript [
   &FN_SetSpawnGroup(Number of Spawn Groups=6)
   &InitializeParticle(
    Color=<rgba 0 2 4 0.5>
    Lifetime=2
    Sprite Rotation Mode=::NewEnumerator2
    Sprite Size Mode=::NewEnumerator1
    Sprite UV Mode=::NewEnumerator3
    Uniform Sprite Size Max=20
    Uniform Sprite Size Min=15
    UsePositionOffset=false
    Write Material Random=true
   )
   &CameraOffset(Camera Offset Amount=-30)
  ]
  particleUpdateScript [
   &ParticleState()
   &DynamicMaterialParameters(Write Parameter Index 0=false)
   /Game/Effects/Niagara/Modules/Sprite/FN_SpriteUV(
    Randomize Sprite UV scale=true
    SubUV Lookup Index:NiagaraFloat=&FloatFromCurve(
     FloatCurve=DataInterface() {propertyValues {Curve {keys [<at 0 0>,<at 1 1>]}}}
     Scale Curve=&MakeCustomFloatFromBool(Bool=&RandomBool(), True Float=1, False Float=-1)
    )
    SubUV Loop Count=2
   )
   &ScaleSpriteSize(Scale Factor=&Vector2DFromFloat(Value=&FloatFromCurve(
      FloatCurve=DataInterface() {propertyValues {Curve {keys [
          {arriveTangent=12.8476, interpMode=RCIM_Cubic, leaveTangent=12.8476, tangentMode=RCTM_User, time=0, value=0.5}
          {arriveTangent=0.0791, interpMode=RCIM_Cubic, leaveTangent=0.0791, tangentMode=RCTM_User, time=0.1, value=1}
          {interpMode=RCIM_Cubic, tangentMode=RCTM_User, time=0.9, value=1}
          {arriveTangent=-11.0136, interpMode=RCIM_Cubic, leaveTangent=-11.0136, tangentMode=RCTM_User, time=1, value=0.5}
         ]}}}
      Scale Curve=&SelectFloatFromArray(Direct Array Index=$Particles.SpawnGroup, Float Selection Array=DataInterface() {propertyValues {FloatData [1,1,1.25,1.25,2,1.5]}})
     )))
   &ScaleColor(
    Scale Alpha=&FloatFromCurve(FloatCurve=DataInterface() {propertyValues {Curve {keys [
         {interpMode=RCIM_Cubic, tangentMode=RCTM_User, time=0, value=0}
         {interpMode=RCIM_Cubic, tangentMode=RCTM_User, time=0.5, value=1}
         {tangentMode=RCTM_User, time=1, value=0}
        ]}}})
    ScaleA=false
    ScaleRGB=true
   )
   &SolveForcesAndVelocity(Clamp Velocity=false, Limit Acceleration=false)
   &FN_BoneFlow(
    Angle:NiagaraFloat=&RandomRangeFloat(Minimum=-1, Maximum=1)
    BoneID:NiagaraInt32=$Particles.SpawnGroup
    Height:NiagaraFloat=&FloatFromCurve(FloatCurve=DataInterface() {propertyValues {Curve {keys [{arriveTangent=1.4394, interpMode=RCIM_Cubic, time=0, value=0}, {interpMode=RCIM_Cubic, tangentMode=RCTM_User, time=1, value=1.05}]}}})
    Interpolate Mesh Sampling=false
    Radius=1
    SkeletalMesh:NiagaraDataInterfaceSkeletalMesh=$User.SkeletalMesh_Flow
    Write Position Attribute=true
   )
  ]
  rendererClasses [&NiagaraSpriteRendererProperties]
  renderers [{rendererClass=&NiagaraSpriteRendererProperties}]
 }
 Flares_Hand {
  emitterUpdateScript [
   &EmitterState(Life Cycle Mode=::NewEnumerator1, Loop Duration=1, UseLoopDelay=false)
   &SpawnRate(SpawnRate=30, Use Spawn Probability=true)
  ]
  particleSpawnScript [
   /Game/Effects/Niagara/Modules/Skydiving/FN_DivingAttributes/FN_DivingAttributes_v2(Accumulate Dive=0.1, Accumulate Rest=-0.1, SkeletalMesh:NiagaraDataInterfaceSkeletalMesh=$User.SkeletalMesh) {enabled=false}
   &FN_SetSpawnGroup()
   /Niagara/Modules/Spawn/Initialization/InitializeParticle(
    Color=<rgba 0 2 4 1>
    Lifetime=0.3
    Module.Write Color=true
    Module.Write Lifetime=true
    Module.Write Mass=false
    Module.Write Material Random=true
    Module.Write Position=false
    Module.Write Scale=false
    Module.Write SpriteRotation=true
    Module.Write SpriteSize=true
    Sprite Rotation=&RandomRangeFloat(Minimum=0, Maximum=360)
    Sprite Size=&Vector2DFromFloat(Value=&RandomRangeFloat(Minimum=100, Maximum=100))
   )
  ]
  particleUpdateScript [
   &ParticleState()
   &SkeletalMeshLocation(
    Bone / Socket Index=$Particles.SpawnGroup
    Bone Sampling Mode=::NewEnumerator4
    Inherit Velocity=false
    MaskSpawnGroup=false
    Sampled Position Coordinate Space=::Local
    Skeletal Mesh=$User.SkeletalMesh
   ) {enabled=false, module=SkeletalMeshLocation001}
   &SkeletalMeshLocation(
    Bone / Socket Index=$Particles.SpawnGroup
    Bone Sampling Mode=::NewEnumerator4
    Inherit Velocity=false
    Interpolate Mesh Sampling=false
    MaskSpawnGroup=false
    Skeletal Mesh=$User.SkeletalMesh
   )
   /Game/Effects/Niagara/Modules/Skydiving/FN_SkydivePositionOffset_v1/FN_SkydivePositionOffset(
    Bone Orientation:Quat4f=$Output.SkeletalMeshLocation.SampledRotation
    Bone Position:Vector3f=$Output.SkeletalMeshLocation.SampledPosition
    Offset:NiagaraFloat=/Game/Effects/Niagara/DynamicInputs/Skydiving/SkydiveHandsFeetFloatModifier(Hands:NiagaraFloat=20, Feet:NiagaraFloat=45)
   ) {enabled=false}
   /Game/Effects/Niagara/Modules/Skydiving/FN_Skydive_Velocity_v4/FN_Skydive_Velocity(
    Bone Orientation:Quat4f=$Output.SkeletalMeshLocation.SampledRotation
    Bone Position:Vector3f=$Output.SkeletalMeshLocation.SampledPosition
    Dive:NiagaraFloat=$System.Dive
    FE Speed=1
    FE Spread=0.1
    SpawnGroup:NiagaraInt32=$Particles.SpawnGroup
    Write to Intrinsic Properties=false
   ) {enabled=false}
   &ScaleSpriteSize(Scale Factor=&Vector2DFromFloat(Value=&FloatFromCurve(FloatCurve=DataInterface() {propertyValues {Curve {keys [
          {arriveTangent=1.5832, interpMode=RCIM_Cubic, time=0, value=0.5}
          {arriveTangent=0.5893, interpMode=RCIM_Cubic, leaveTangent=0.5893, tangentMode=RCTM_User, time=1, value=1}
         ]}}})))
   &SolveForcesAndVelocity(Clamp Velocity=false, Limit Acceleration=false)
   &SpriteRotationRate(Rotation Rate=&MakeCustomFloatFromBool(Bool=&RandomBool(), True Float=10, False Float=-10)) {enabled=false}
   &ScaleColor(
    Scale Alpha=&FloatFromCurve(FloatCurve=DataInterface() {propertyValues {Curve {keys [
         {arriveTangent=4.9245, interpMode=RCIM_Cubic, time=0, value=0}
         <at 0.33 1 RCIM_Cubic>
         {interpMode=RCIM_Cubic, leaveTangent=-1.9375, time=1, value=0}
        ]}}})
    Scale RGB=&VectorFromFloat(Value=&FloatFromCurve(
      FloatCurve=DataInterface() {propertyValues {Curve {keys [
          {arriveTangent=8517.9209, interpMode=RCIM_Cubic, leaveTangent=2.9045, tangentMode=RCTM_User, time=0, value=0.1}
          <at 0.33 1 RCIM_Cubic>
          {interpMode=RCIM_Cubic, leaveTangent=-4.1243, time=1, value=0.1}
         ]}}}
      Scale Curve=&Multiply_Float(A=&FNTimeOfDayScalar(Night Scale=0.25, Morning Scale=0.5, Evening Scale=0.5), B=1)
     ))
    ScaleA=true
    ScaleRGB=true
   )
   &CameraOffset(Camera Offset Amount=-10)
   &DynamicMaterialParameters(
    Index 0 Param 1=0.6
    Index 0 Param 2=1
    Module.Param0WriteEnabled=true
    Module.Param1WriteEnabled=true
    Module.Param2WriteEnabled=true
    Module.Param3WriteEnabled=true
   ) {enabled=false}
  ]
  rendererClasses [&NiagaraSpriteRendererProperties]
  renderers [{rendererClass=&NiagaraSpriteRendererProperties}]
 }
 Light {
  emitterUpdateScript [
   &EmitterState(
    Inactive Response=::NewEnumerator1
    Life Cycle Mode=::NewEnumerator1
    Loop Behavior=::NewEnumerator1
    Loop Duration Mode=::NewEnumerator1
    UseLoopDelay=false
   )
   &SpawnBurst_Instantaneous(Spawn Count=2, Use Spawn Probability=false)
  ]
  particleSpawnScript [
   &FN_SetSpawnGroup(Number of Spawn Groups=2)
   /Niagara/Modules/Spawn/Initialization/V2/InitializeParticle(Color=<rgba 2.5 20 25 1>, UsePositionOffset=false, Write Material Random=true)
  ]
  particleUpdateScript [
   &ParticleState(Kill Particles When Lifetime Has Elapsed=false, Let Infinitely Lived Particles Die When Emitter Deactivates=true)
   &SkeletalMeshLocation(
    Bone / Socket Index=14
    Bone Sampling Mode=::NewEnumerator4
    Inherit Velocity=false
    Interpolate Mesh Sampling=false
    MaskSpawnGroup=false
    Sampled Position Offset=&MakeVector(
     X=&Sine(Normalized Angle=$System.FN_Integrator.Value, Period=0.6, Scale=50, Bias=0)
     Y=&Cosine(Normalized Angle=$System.FN_Integrator.Value, Period=0.6, Scale=50, Bias=0)
     Z=20
    )
    Sampled Position Offset Coordinate Space:ENiagaraMeshTransforms=::NewEnumerator2
    Skeletal Mesh=$User.SkeletalMesh_Flow
   )
   &ScaleColor(
    Scale Alpha=&FNTimeOfDayScalar(Night Scale=0.5, Morning Scale=0.7, Evening Scale=0.7)
    Scale RGB=&VectorFromFloat(Value=&Lerp_Float(A=0.5, B=1, Alpha=$System.Dive))
    ScaleA=true
    ScaleRGB=true
   ) {module=ScaleColor001}
  ]
  rendererClasses [&NiagaraLightRendererProperties]
  renderers [{rendererClass=&NiagaraLightRendererProperties, rendererIndex=0}]
 }
 Streaks_Surround {
  emitterUpdateScript [
   &EmitterState(Life Cycle Mode=::NewEnumerator1, Loop Duration=1, UseLoopDelay=false)
   &SpawnRate(SpawnRate=60, Use Spawn Probability=false)
  ]
  particleSpawnScript [
   /Niagara/Modules/Spawn/Initialization/InitializeParticle(
    Color=<rgba 0 3.4803 5 1>
    Lifetime=3
    Module.Write Color=true
    Module.Write Lifetime=true
    Module.Write Mass=true
    Module.Write Material Random=true
    Module.Write Position=false
    Module.Write Scale=false
    Module.Write SpriteRotation=false
    Module.Write SpriteSize=true
    Position=&SimulationPosition()
    Sprite Size=&MaskVector2dBySpawnGroup(
     Masked Vector2D=&RandomRangeVector2D(Minimum=<1 30>, Maximum=<15 40>)
     Passthrough Vector2D=&RandomRangeVector2D(Minimum=<1 25>, Maximum=<5 35>)
     Spawn Group Mask=2
    )
   ) {enabled=false}
   &DynamicMaterialParameters(
    Index 0 Param 1=&RandomRangeFloat(Minimum=-0.1, Maximum=0.1)
    Module.Param0WriteEnabled=true
    Module.Param1WriteEnabled=false
    Module.Param2WriteEnabled=false
    Module.Param3WriteEnabled=false
   )
   /Niagara/Modules/Spawn/Initialization/V2/InitializeParticle(
    Color=<rgba 0 3.4803 5 0.5>
    Lifetime Max=0.3
    Lifetime Min=0.2
    Lifetime Mode=::NewEnumerator1
    Sprite Size=<10 60>
    Sprite Size Mode=::NewEnumerator4
    UsePositionOffset=false
    Write Material Random=true
   ) {module=InitializeParticle001}
   &SphereLocation(
    MaskSpawnGroup=false
    Override Seed=false
    Sphere Coordinate Space=::Simulation
    Sphere Radius=300
    Surface Only=true
    Surface Only Band Thickness=100
    UseNonUniformScale=false
    UseOffset=true
    UseSphereOrientation=false
   )
  ]
  particleUpdateScript [
   &AddVelocity(Velocity=&MakeVector(X=0, Y=0, Z=&RandomRangeFloat(Minimum=9, Maximum=15))) {enabled=false}
   &InheritSourceMovement(
    Applied Force Scale=&VectorFromFloat(Value=0.75)
    Applied Position Scale=<0.5 0.5 0.5>
    ApplyForce=true
    ApplyForceFalloff=false
    ApplyPosition=true
    ApplyPositionFalloff=false
    Module.LimitForceSpeed=false
    Source Velocity:Vector3f=$System.UFOVelocity
   )
   &ParticleState()
   &SolveForcesAndVelocity(Clamp Velocity=false, Limit Acceleration=false)
   &ScaleColor(
    Scale Alpha=&FloatFromCurve(FloatCurve=DataInterface() {propertyValues {Curve {keys [
         {interpMode=RCIM_Cubic, tangentMode=RCTM_User, time=0, value=0}
         {interpMode=RCIM_Cubic, tangentMode=RCTM_User, time=0.5, value=1}
         {tangentMode=RCTM_User, time=1, value=0}
        ]}}})
    Scale RGB=&MakeVectorFromLinearColorRGB(Color=<rgba 1 1 1 1>)
    ScaleA=true
    ScaleRGB=false
   )
   &ScaleSpriteSize(Scale Factor=&MakeVector2D(
     X=1
     Y=&FloatFromCurve(FloatCurve=DataInterface() {propertyValues {Curve {keys [<at 0 0>,<at 1 1>]}}}, Scale Curve=4)
    ))
   &Color(Color=&MaskLinearColorBySpawnGroup(
     Masked Linear Color=<rgba 1 0 0.2339 1>
     Passthrough Linear Color=&MaskLinearColorBySpawnGroup(Masked Linear Color=<rgba 0.1072 1 0 1>, Passthrough Linear Color=<rgba 0 0.449 1 1>, Spawn Group Mask=1)
     Spawn Group Mask=2
    )) {enabled=false}
  ]
  rendererClasses [&NiagaraSpriteRendererProperties]
  renderers [{rendererClass=&NiagaraSpriteRendererProperties}]
 }
 Tractor_Beam_Rings {
  emitterUpdateScript [
   &EmitterState(Life Cycle Mode=::NewEnumerator1, UseLoopDelay=false)
   &SpawnRate(SpawnRate=&Lerp_Float(A=24, B=48, Alpha=$System.Dive), Use Spawn Probability=true)
  ]
  particleSpawnScript [
   &FN_SetSpawnGroup(Number of Spawn Groups:NiagaraInt32=$System.SpawnGroup)
   &InitializeParticle(
    Color=<rgba 0 1.3921 2 0.5>
    Lifetime=&Lerp_Float(A=1.2, B=0.6, Alpha=$System.Dive)
    Mesh Scale Mode=::NewEnumerator3
    Mesh Uniform Scale=0.075
    UsePositionOffset=false
    Write Material Random=true
   ) {module=InitializeParticle001}
   &InitialMeshOrientation(
    Orientation Axis=<0 0 1>
    Orientation Coordinate Space=::Local
    Orientation Vector=<0 0 1>
    Rotation=&MakeVector(X=0, Y=0, Z=&RandomRangeFloat(Minimum=0, Maximum=360))
    Rotation Coordinate Space=::NewEnumerator0
    Use Orientation Vector=false
    Use Rotation Vector=true
   )
  ]
  particleUpdateScript [
   &ParticleState()
   &SkeletalMeshLocation(
    Bone / Socket Index=$Particles.SpawnGroup
    Bone Sampling Mode=::NewEnumerator4
    Inherit Velocity=false
    Interpolate Mesh Sampling=false
    MaskSpawnGroup=false
    Sampled Position Offset=&Multiply_VectorByFloat(
     Vector=$System.UfoVelocity
     Float=&FloatFromCurve(FloatCurve=DataInterface() {propertyValues {Curve {keys [<at 0 0>,<at 1 1>]}}}, Scale Curve=$System.UFODistance)
    )
    Skeletal Mesh=$User.SkeletalMesh
   )
   /SetVariables(Particles.MeshOrientation=&MultiplyQuaternion(
     Quaternion A=&FindQuatBetweenTwoVectors(From Vector=<0 0 -1>, To Vector=$System.UfoVelocity)
     Quaternion B=$Particles.Initial.MeshOrientation
    )) {isSetParameters=true, module=SetVariables_#5}
   &SolveForcesAndVelocity(Clamp Velocity=false, Limit Acceleration=false)
   &ScaleMeshSize(Scale Factor=&VectorFromCurve(VectorCurve=DataInterface() {propertyValues {XCurve {keys [<at 0 1>,<at 1 0.5>]}, YCurve {keys [<at 0 1>,<at 1 0.5>]}, ZCurve {keys [<at 0 0.5>,<at 1 0.5>]}}}))
   &ScaleColor(
    Scale Alpha=&FloatFromCurve(FloatCurve=DataInterface() {propertyValues {Curve {keys [<at 0 0 RCIM_Cubic>,<at 0.85 1 RCIM_Cubic>,<at 1 0 RCIM_Cubic>]}}})
    Scale RGB=&VectorFromFloat(Value=&FloatFromCurve(FloatCurve=DataInterface() {propertyValues {Curve {keys [<at 0 0.5>,<at 0.85 1>,<at 1 0.5>]}}}))
    ScaleA=true
    ScaleRGB=true
   )
   &DynamicMaterialParameters(
    Index 0 Param 1=0.25
    Index 0 Param 2=1
    Index 0 Param 3=1
    Index 0 Param 4=0.15
    Module.Param0WriteEnabled=false
    Module.Param1WriteEnabled=false
    Module.Param2WriteEnabled=true
    Module.Param3WriteEnabled=true
   )
  ]
  rendererClasses [&NiagaraMeshRendererProperties]
  renderers [{rendererClass=&NiagaraMeshRendererProperties, rendererIndex=0}]
 }
 UFO {
  emitterUpdateScript [
   &EmitterState(
    Inactive Response=::NewEnumerator1
    Life Cycle Mode=::NewEnumerator1
    Loop Behavior=::NewEnumerator1
    Loop Duration Mode=::NewEnumerator1
    UseLoopDelay=false
   )
   &SpawnBurst_Instantaneous(Spawn Count=4, Use Spawn Probability=false)
  ]
  particleSpawnScript [
   &FN_SetSpawnGroup(Number of Spawn Groups:NiagaraInt32=$System.SpawnGroup)
   &InitializeParticle(
    Mesh Scale Mode=::NewEnumerator3
    Mesh Uniform Scale=0.15
    Sprite Rotation Mode=::NewEnumerator2
    UsePositionOffset=false
    Write Material Random=true
   ) {module=InitializeParticle001}
   /SetVariables(Particles.LightColor=<rgba 0 3.4803 5 0.5>) {isSetParameters=true}
   &InitialMeshOrientation(
    Orientation Axis=<0 0 1>
    Orientation Coordinate Space=::Local
    Orientation Vector=<0 0 1>
    Rotation=&MakeVector(X=0, Y=0, Z=&RandomRangeFloat(Minimum=0, Maximum=1))
    Rotation Coordinate Space=::NewEnumerator0
    Use Orientation Vector=false
    Use Rotation Vector=true
   )
  ]
  particleUpdateScript [
   &ParticleState(Kill Particles When Lifetime Has Elapsed=false, Let Infinitely Lived Particles Die When Emitter Deactivates=true)
   &SkeletalMeshLocation(
    Bone / Socket Index=$Particles.SpawnGroup
    Bone Sampling Mode=::NewEnumerator4
    Inherit Velocity=false
    MaskSpawnGroup=false
    Sampled Position Offset=&Multiply_VectorByFloat(
     Vector=$System.UfoVelocity
     Float=&Add_Float(
      A=$System.UFODistance
      B=&Sine(
       Normalized Angle=&Add_Float(A=$System.FN_Integrator.Value, B=&RandomRangeFloat(Minimum=0, Maximum=1))
       Period=0.8
       Scale=2
       Bias=0
      )
     )
    )
    Skeletal Mesh=$User.SkeletalMesh
   )
   /SetVariables(Particles.MeshOrientation=&MultiplyQuaternion(
     Quaternion A=&FindQuatBetweenTwoVectors(
      From Vector=&MakeVector(
       X=&Sine(
        Normalized Angle=&Add_Float(A=$System.FN_Integrator.Value, B=&RandomRangeFloat(Minimum=0, Maximum=1))
        Period=&MakeCustomFloatFromBool(Bool=&RandomBool(), True Float=1.5, False Float=-1.5)
        Scale=1
        Bias=0
       )
       Y=&Cosine(
        Normalized Angle=&Add_Float(A=$System.FN_Integrator.Value, B=&RandomRangeFloat(Minimum=0, Maximum=1))
        Period=&MakeCustomFloatFromBool(Bool=&RandomBool(), True Float=1.5, False Float=-1.5)
        Scale=1
        Bias=0
       )
       Z=&RandomRangeFloat(Minimum=5, Maximum=6)
      )
      To Vector=<0 0 1>
     )
     Quaternion B=$Particles.Initial.MeshOrientation
    )) {isSetParameters=true, module=SetVariables_#5}
   /SetVariables(Particles.MeshOrientation=&MultiplyQuaternion(
     Quaternion A=&FindQuatBetweenTwoVectors(From Vector=<0 0 -1>, To Vector=$System.UfoVelocity)
     Quaternion B=$Particles.MeshOrientation
    )) {isSetParameters=true, module=SetVariables_#6}
   &SolveForcesAndVelocity(Clamp Velocity=false, Limit Acceleration=false)
   &DynamicMaterialParameters(
    Index 0 Param 1=1
    Index 0 Param 2=1
    Index 0 Param 3=1
    Index 0 Param 4=0.15
    Module.Param0WriteEnabled=true
    Module.Param1WriteEnabled=false
    Module.Param2WriteEnabled=true
    Module.Param3WriteEnabled=true
   )
  ]
  rendererClasses [&NiagaraMeshRendererProperties]
  renderers [
   {rendererClass=&NiagaraMeshRendererProperties, rendererIndex=0}
   {rendererClass=&NiagaraMeshRendererProperties, rendererIndex=1}
  ]
 }
 UFO_Small {
  emitterUpdateScript [
   &EmitterState(
    Inactive Response=::NewEnumerator1
    Life Cycle Mode=::NewEnumerator1
    Loop Behavior=::NewEnumerator1
    Loop Duration Mode=::NewEnumerator1
    UseLoopDelay=false
   )
   &SpawnBurst_Instantaneous(Spawn Count=4, Use Spawn Probability=false) {enabled=false}
   &SpawnRate(Spawn Probability=0.6, SpawnRate=15, Use Spawn Probability=true) {enabled=false}
   &SpawnParticlesFromOtherEmitter(
    Attribute Reader=DataInterface() {propertyValues {EmitterBinding {bindingMode=Other, emitterName=UFO}}}
    Calculate Spawn Rate Per Particle=false
    Spawn Rate=&RandomRangeFloat(Minimum=5, Maximum=15, Evaluation Type=::NewEnumerator1)
    Use Maximum Age=false
    Use Minimum Age=false
   )
  ]
  particleSpawnScript [
   &FN_SetSpawnGroup(Number of Spawn Groups:NiagaraInt32=$System.SpawnGroup)
   &InitializeParticle(
    Lifetime=&RandomRangeFloat(Minimum=0.75, Maximum=1)
    Mesh Scale=&VectorFromFloat(Value=0.075)
    Mesh Scale Mode=::NewEnumerator4
    Sprite Rotation Mode=::NewEnumerator2
    UsePositionOffset=false
    Write Material Random=true
   ) {module=InitializeParticle001}
   /SetVariables(Particles.LightColor=<rgba 0 3.4803 5 0.5>) {isSetParameters=true}
   &SampleParticlesFromOtherEmitter()
   &InitialMeshOrientation(
    Orientation Axis=<0 0 1>
    Orientation Coordinate Space=::Local
    Orientation Vector=<0 0 1>
    Rotation=&MakeVector(X=0, Y=0, Z=&RandomRangeFloat(Minimum=0, Maximum=0))
    Rotation Coordinate Space=::NewEnumerator0
    Use Orientation Vector=false
    Use Rotation Vector=true
   )
   &AddVelocity(
    Scale Added Velocity=<0.5 0.5 1>
    Velocity=&RandomRangeVector(Minimum=<-100 -100 0>, Maximum=<100 100 0>)
   )
   &SkeletalMeshLocation(
    Bone / Socket Index=$Particles.SpawnGroup
    Bone Sampling Mode=::NewEnumerator4
    Inherit Velocity=false
    MaskSpawnGroup=false
    Sampled Position Offset=&Multiply_VectorByFloat(Vector=$System.UfoVelocity, Float=&Add_Float(A=$System.UFODistance, B=-20))
    Skeletal Mesh=$User.SkeletalMesh
   ) {enabled=false}
  ]
  particleUpdateScript [
   &ParticleState()
   /SetVariables(Particles.MeshOrientation=&MultiplyQuaternion(
     Quaternion A=&FindQuatBetweenTwoVectors(
      From Vector=&MakeVector(
       X=&Sine(
        Normalized Angle=&Add_Float(A=$Emitter.Age, B=&RandomRangeFloat(Minimum=0, Maximum=1))
        Period=&MakeCustomFloatFromBool(Bool=&RandomBool(), True Float=1.5, False Float=-1.5)
        Scale=1
        Bias=0
       )
       Y=&Cosine(
        Normalized Angle=&Add_Float(A=$Emitter.Age, B=&RandomRangeFloat(Minimum=0, Maximum=1))
        Period=&MakeCustomFloatFromBool(Bool=&RandomBool(), True Float=1.5, False Float=-1.5)
        Scale=1
        Bias=0
       )
       Z=&RandomRangeFloat(Minimum=4, Maximum=5)
      )
      To Vector=<0 0 1>
     )
     Quaternion B=$Particles.Initial.MeshOrientation
    )) {isSetParameters=true, module=SetVariables_#5}
   /SetVariables(Particles.MeshOrientation=&MultiplyQuaternion(
     Quaternion A=&FindQuatBetweenTwoVectors(From Vector=<0 0 -1>, To Vector=$System.UfoVelocity)
     Quaternion B=$Particles.MeshOrientation
    )) {isSetParameters=true, module=SetVariables_#6}
   /SetVariables(
    Particles.UFOPivotOffset=3
    Particles.UFOSpinSpeed:NiagaraFloat=&Multiply_Float(
     A=0.5
     B=&MakeCustomFloatFromBool(Bool=&RandomBool(), True Float=1, False Float=-1)
    )
   ) {isSetParameters=true}
   /SetVariables(Particles.Position=&Add_Vector(
     A=Linked() {linkedVariable {name=Particles.Position, type=NiagaraPosition}}
     B=&Multiply_VectorByFloat(
      Vector=&MakeVector(
       X=&Sine(Normalized Angle=$Emitter.Age, Period=$Particles.UFOSpinSpeed, Scale=$Particles.UFOPivotOffset, Bias=0)
       Y=&Cosine(Normalized Angle=$Emitter.Age, Period=$Particles.UFOSpinSpeed, Scale=$Particles.UFOPivotOffset, Bias=0)
       Z=&FloatFromCurve(
        FloatCurve=DataInterface() {propertyValues {Curve {keys [
            {arriveTangent=-1.4188, interpMode=RCIM_Cubic, leaveTangent=-1.4188, tangentMode=RCTM_User, time=0, value=0}
            {arriveTangent=1, interpMode=RCIM_Cubic, leaveTangent=1, time=0.1, value=-0.1}
            {interpMode=RCIM_Cubic, leaveTangent=1.4447, time=1, value=1}
           ]}}}
        Scale Curve=10
       )
      )
      Float=&FloatFromCurve(FloatCurve=DataInterface() {propertyValues {Curve {keys [{tangentMode=RCTM_User, time=0, value=0}, {tangentMode=RCTM_User, time=1, value=1}]}}})
     )
    )) {isSetParameters=true}
   &InheritSourceMovement(
    Applied Force Scale=&VectorFromFloat(Value=-0.5)
    Applied Position Scale=<0.9 0.9 0.9>
    ApplyForce=false
    ApplyPosition=true
    ApplyPositionFalloff=false
   )
   &SolveForcesAndVelocity(Clamp Velocity=false, Limit Acceleration=false)
   &ScaleColor(
    Linear Color Curve=DataInterface() {propertyValues {
      AlphaCurve {keys [<at 0 0>,<at 0.0712 1>,<at 0.8886 1>,<at 1 0>]}
      BlueCurve {keys [<at 0.1009 10>,<at 0.2946 1>,<at 0.6927 1>,<at 0.8868 10>]}
      GreenCurve {keys [<at 0.1009 6.9607>,<at 0.2946 1>,<at 0.6927 1>,<at 0.8868 6.9607>]}
      RedCurve {keys [<at 0.1009 0>,<at 0.2946 1>,<at 0.6927 1>,<at 0.8868 0>]}
     }}
    Scale Mode=::NewEnumerator2
   )
   &ScaleMeshSize(Scale Factor=&VectorFromFloat(Value=&FloatFromCurve(FloatCurve=DataInterface() {propertyValues {Curve {keys [
          {arriveTangent=2.0977, interpMode=RCIM_Cubic, leaveTangent=2.0977, tangentMode=RCTM_User, time=0, value=0}
          {interpMode=RCIM_Cubic, tangentMode=RCTM_User, time=0.3, value=1}
          {arriveTangent=0.0772, interpMode=RCIM_Cubic, leaveTangent=0.0772, tangentMode=RCTM_User, time=0.8, value=1}
          {arriveTangent=-5.7608, interpMode=RCIM_Cubic, leaveTangent=-5.7608, tangentMode=RCTM_User, time=1, value=0}
         ]}}})))
   &DynamicMaterialParameters(
    Index 0 Param 1=1
    Index 0 Param 2=1
    Index 0 Param 3=1
    Index 0 Param 4=0.15
    Module.Param0WriteEnabled=true
    Module.Param1WriteEnabled=false
    Module.Param2WriteEnabled=true
    Module.Param3WriteEnabled=true
   )
  ]
  rendererClasses [&NiagaraMeshRendererProperties]
  renderers [
   {rendererClass=&NiagaraMeshRendererProperties, rendererIndex=0}
   {rendererClass=&NiagaraMeshRendererProperties, rendererIndex=1}
  ]
 }
}
summary {
 systemName=NS_Contrail_Abduction
 emitters [
  {emitterName=Aura, rendererClasses [&NiagaraSpriteRendererProperties]}
  {emitterName=Flares_Hand, rendererClasses [&NiagaraSpriteRendererProperties]}
  {emitterName=Tractor_Beam_Rings, rendererClasses [&NiagaraMeshRendererProperties]}
  {emitterName=Light, rendererClasses [&NiagaraLightRendererProperties]}
  {emitterName=Streaks_Surround, rendererClasses [&NiagaraSpriteRendererProperties]}
  {emitterName=UFO, rendererClasses [&NiagaraMeshRendererProperties]}
  {emitterName=UFO_Small, rendererClasses [&NiagaraMeshRendererProperties]}
 ]
 userVariables [
  User.bIsFrontEnd {type=NiagaraBool}
  User.bIsFrontEndPreview {type=NiagaraBool, defaultValue=true}
  User.ParaGlideLeanAlpha {type=NiagaraFloat, defaultValue=0, description=""}
  User.SkeletalMesh {
   type=NiagaraDataInterfaceSkeletalMesh
   defaultValue {
    struct=NiagaraExt_VariableValue_DataInterface
    value {dataInterface=&self:NiagaraDataInterfaceSkeletalMesh_0}
   }
  }
  User.SkeletalMesh_Flow {
   type=NiagaraDataInterfaceSkeletalMesh
   defaultValue {
    struct=NiagaraExt_VariableValue_DataInterface
    value {dataInterface=&self:NiagaraDataInterfaceSkeletalMesh_2}
   }
  }
 ]
}
